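# Additive, unlit mesh-particle material for lightning at feet on the Grind Trap device, built on SplineThicken for width.
# World position offset adds stepped, panning random noise, masked toward V=0, with timing and V scale set by particle dynamic parameters.
# Emissive multiplies a scrolling ring texture by sine/cosine flicker, a mask toward V=1 and particle color, oversaturated through negative desaturation except where green is high.
# The LightningColorA parameter is unconnected.
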Material /Game/Creative/Devices/GrindTrap/Materials/M_GrindTrap_FeetLightning
graph {
 properties {
  bEnableResponsiveAA=true
  bUsedWithMeshParticles=true
  blendMode=BLEND_Additive
  pixelDepthOffsetMode=PDOM_Legacy
  shadingModel=MSM_Unlit
  twoSided=true
  usedShadingModels=Unlit
 }
 nodes [
  n1 = ScalarParameter() {defaultValue=10, parameterName="WidthBase (S)"}
  n2 = Add(
   A=DynamicParameter() {paramNames [Random,Speed,VScale,Param4]}.Random
   B=Multiply(A=Time() {bOverride_Period=true, period=1}, B=DynamicParameter() {paramNames [Random,Speed,VScale,Param4]}.Speed)
  )
  n3 = ScalarParameter() {defaultValue=6, parameterName=WPONoisePanSteps_V}
  n4 = Floor(None=Multiply(A=@n2, B=@n3))
  n5 = Constant() {r=1}
  n6 = DynamicParameter() {paramNames [Random,Speed,VScale,Param4]}
  n7 = Add(A=AppendVector(A=Constant(), B=Divide(A=@n4, B=@n3)), B=Multiply(A=TextureCoordinate(), B=AppendVector(A=@n5, B=@n6.VScale)))
  n8 = Multiply(A=@n7, B=AppendVector(A=Constant() {r=1}, B=@n3))
  n9 = MaterialFunctionCall!(PrimaryTextureCoordinates=@n8, 2ndOrderTextureCoordinates, RandomSeed, NoiseDensityBias, Bypass Noise Influence Crosstalk) {materialFunction=/Game/Effects/Fort_Effects/Materials/_Functions/MF_RandomUVSample}
  n10 = MaterialFunctionCall(PrimaryTextureCoordinates=@n8, RandomSeed=Constant() {r=2}) {materialFunction=&MF_RandomUVSample}
  n11 = MaterialFunctionCall(PrimaryTextureCoordinates=@n8, RandomSeed=Constant() {r=5}) {materialFunction=&MF_RandomUVSample}
  n12 = Transform(None=ConstantBiasScale(None=MaterialFunctionCall(X=@n9.NoiseWithRange, Y=@n10.NoiseWithRange, Z=@n11.NoiseWithRange) {materialFunction=&MakeFloat3}.Result) {bias=-0.5, scale=2}) {transformSourceType=TRANSFORMSOURCE_World, transformType=TRANSFORM_Tangent}
  n13 = VectorParameter() {parameterName=wpoNoiseScale, defaultValue=<rgba 8 8 8 0>}
  n14 = ComponentMask(None=TextureCoordinate()) {g=true}
  n15 = Constant()
  n16 = Constant() {r=0.9}
  n17 = Constant() {r=0.5}
  Add(
   A=MaterialFunctionCall(WidthBase=@n1, WidthTip=@n1) {materialFunction=&SplineThicken}.WorldPositionOffset
   B=Multiply(A=Multiply(A=@n12, B=@n13.RGB), B=SphereMask(A=@n14, B=@n15, Radius=@n16, Hardness=@n17))
  )
  n18 = ConstantBiasScale(None=Add(A=Sine(None=@n2) {period=0.1133}, B=Cosine(None=@n2) {period=0.0874})) {bias=2, scale=5}
  n19 = TextureSample(UVs=@n7) {samplerType=SAMPLERTYPE_LinearColor, texture=/Game/Effects/Fort_Effects/Textures/Tile/T_AMR_MuzzleRing_01}
  n20 = ComponentMask(None=TextureCoordinate()) {g=true}
  n21 = Constant() {r=1}
  n22 = Constant() {r=0.9}
  n23 = Constant() {r=0.5}
  n24 = Multiply(A=Multiply(A=Multiply(A=@n18, B=@n19.R), B=SphereMask(A=@n20, B=@n21, Radius=@n22, Hardness=@n23)), B=ParticleColor().RGB)
  n25 = ScalarParameter() {defaultValue=-20, parameterName=Fraction}
  MaterialFunctionCall(Emissive=LinearInterpolate(A=Desaturation(None=@n24, Fraction=@n25), B=@n24, Alpha=Power(Base=@n19.G) {constExponent=6})) {materialFunction=&MF_FX_EmissiveAdjustments}
  VectorParameter() {parameterName=LightningColorA}
 ]
}
