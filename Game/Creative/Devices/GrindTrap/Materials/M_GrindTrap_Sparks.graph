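# Unlit translucent sprite material for the Grind Trap device's spark particles, drawing a procedural, noise-warped glow.
# Swirl-noise red, with per-particle random tiling and offset, distorts U before UVs are centered and mirrored.
# Cubing the V distance stretches a soft SphereMask along that axis, and exponents sharpen it.
# The mask blends two colors for emissive, tinted by particle color, and combines with particle and parameter alpha for opacity.
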Material /Game/Creative/Devices/GrindTrap/Materials/M_GrindTrap_Sparks
graph {
 properties {
  bUsedWithNiagaraSprites=true
  bUsedWithParticleSprites=true
  blendMode=BLEND_Translucent
  pixelDepthOffsetMode=PDOM_Legacy
  shadingModel=MSM_Unlit
  usedShadingModels=Unlit
 }
 nodes [
  n1 = VectorParameter() {parameterName=LightningColorA, defaultValue=<rgba 1 1 1 0>}
  n2 = VectorParameter() {parameterName=LightningColorB, defaultValue=<rgba 35 35 35 0>}
  n3 = ScalarParameter() {defaultValue=0.2, parameterName=NoiseTileLow}
  n4 = ScalarParameter() {defaultValue=0.5, parameterName=NoiseTileHigh}
  n5 = ParticleRandom()
  n6 = TextureCoordinate()
  n7 = TextureSample(UVs=Multiply(A=LinearInterpolate(A=@n3, B=@n4, Alpha=@n5), B=Add(A=@n5, B=@n6))) {samplerType=SAMPLERTYPE_Normal, texture=/Game/Global/Textures/Default/Noise/T_SwirlNoise_N}
  n8 = Constant()
  n9 = Add(A=TextureCoordinate(), B=Multiply(A=ScalarParameter() {defaultValue=0.1, parameterName=UVNoiseScale}, B=AppendVector(A=@n7.R, B=@n8)))
  n10 = MaterialFunctionCall(Float2=Abs(None=ConstantBiasScale(None=@n9) {bias=-0.5, scale=2})) {materialFunction=&BreakOutFloat2Components}
  n11 = Power(Base=@n10.G) {constExponent=3}
  n12 = Power(
   Base=SphereMask(
    A=MaterialFunctionCall(X=@n10.R, Y=@n11) {materialFunction=&MakeFloat2}.Result
    B=Constant()
    Radius=Constant() {r=1}
    Hardness=Constant()
   )
   Exp=ScalarParameter() {defaultValue=5.5, parameterName=OpacityExponent}
  ) {constExponent=3}
  n13 = ParticleColor()
  MaterialFunctionCall(Emissive=Multiply(A=LinearInterpolate(A=@n1.RGB, B=@n2.RGB, Alpha=@n12), B=@n13.RGB)) {materialFunction=&MF_FX_EmissiveAdjustments}
  Multiply(A=Multiply(A=@n13.A, B=@n12), B=ScalarParameter() {defaultValue=1, parameterName=Alpha})
 ]
}
