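# Additive, unlit particle sprite material for streak effects on the Creative GrindTrap device.
# An explosion streak texture blends between the LightningColorA and LightningColorB parameters, then masks that result to give a two-tone gradient by intensity.
# Particle color RGB tints the result, and a DepthFade with adjustable distance softens intersections by scaling the color itself.
# The faded color feeds the shared MF_FX_EmissiveAdjustments function.
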Material /Game/Creative/Devices/GrindTrap/Materials/M_GrindTrap_Streaks
graph {
 properties {bUsedWithParticleSprites=true, blendMode=BLEND_Additive, pixelDepthOffsetMode=PDOM_Legacy, shadingModel=MSM_Unlit, usedShadingModels=Unlit}
 nodes [
  n1 = TextureSample() {texture=/Game/Effects/Fort_Effects/Gameplay/SmokeDustExplosions/Textures/T_Explosion_Streak_02}
  n2 = Multiply(
   A=LinearInterpolate(
    A=VectorParameter() {parameterName=LightningColorA, defaultValue=<rgba 1 1 1 0>}.RGB
    B=VectorParameter() {parameterName=LightningColorB, defaultValue=<rgba 4 4 4 0>}.RGB
    Alpha=@n1.RGB
   )
   B=@n1.RGB
  )
  n3 = ParticleColor()
  MaterialFunctionCall(Emissive=DepthFade(Opacity=Multiply(A=@n2, B=@n3.RGB), FadeDistance=ScalarParameter() {defaultValue=5, parameterName=FadeDistance})) {materialFunction=&MF_FX_EmissiveAdjustments}
 ]
}
